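# Material function that outputs a repeating sawtooth from time, for cyclic effects and debugging.
# Time times a Speed input passes through Frac, then inverse-lerps between the Fraction at 0 and Fraction at 1 inputs and saturates.
# This holds the output at 0 and 1 at the ends of each cycle, with defaults giving a ramp over the middle half.
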MaterialFunction /Game/Effects/Fort_Effects/Materials/_Functions/MF_FracTime
graph {
 properties {description=Sawtooth output for cyclic time effects and debugging purposes}
 nodes [
  n1 = Constant() {r=0.25}
  n2 = FunctionInput(Preview=@n1) {bUsePreviewValueAsDefault=true, inputName=Fraction at 0, inputType=FunctionInput_Scalar, sortPriority=1}
  n3 = FunctionInput(Preview=OneMinus(None=@n1)) {bUsePreviewValueAsDefault=true, inputName=Fraction at 1, inputType=FunctionInput_Scalar, sortPriority=2}
  n4 = Constant() {r=1}
  n5 = Frac(None=Multiply(A=Time(), B=FunctionInput(Preview=@n4) {bUsePreviewValueAsDefault=true, inputName=Speed, inputType=FunctionInput_Scalar}))
  FunctionOutput(None=Saturate(None=InverseLinearInterpolate(A=@n2, B=@n3, Value=@n5)))
 ]
}
