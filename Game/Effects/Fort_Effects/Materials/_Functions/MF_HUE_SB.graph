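# MF_HUE_SB is a library-exposed effects material function that adjusts the saturation and brightness of an input color or texture.
# It lerps between a grey value and the original color by the HUE Saturation input, then multiplies by HUE Brightness.
# The grey is a dot product with (1,1,1), an unweighted, unnormalized channel sum, so desaturation also brightens.
# Despite the name, it has no hue-shift control.
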MaterialFunction /Game/Effects/Fort_Effects/Materials/_Functions/MF_HUE_SB
graph {
 properties {bExposeToLibrary=true}
 nodes [
  n1 = FunctionInput() {inputName=Color or Texture}
  n2 = DotProduct(A=@n1, B=Constant3Vector() {constant=<rgba 1 1 1 0>})
  n3 = FunctionInput() {inputName=HUE Saturation, inputType=FunctionInput_Scalar, previewValue=<1 1 1 1>}
  FunctionOutput(None=Multiply(
    A=LinearInterpolate(A=@n2, B=@n1, Alpha=@n3)
    B=FunctionInput() {inputName=HUE Brightness, inputType=FunctionInput_Scalar, previewValue=<1 1 1 1>}
   ))
 ]
}
