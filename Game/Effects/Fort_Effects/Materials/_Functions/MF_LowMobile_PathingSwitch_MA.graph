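# MF_LowMobile_PathingSwitch_MA is a library-exposed material function for effects materials that selects between two Material Attributes inputs, Default and Low/Mobile.
# A Quality Switch picks the Low/Mobile input at low quality and the Default input otherwise.
# A Shading Path Switch wraps that result and forces the Low/Mobile input on the mobile shading path, so one cheaper attribute set serves both cases.
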MaterialFunction /Game/Effects/Fort_Effects/Materials/_Functions/MF_LowMobile_PathingSwitch_MA
graph {
 properties {bExposeToLibrary=true}
 nodes [
  n1 = FunctionInput() {inputName=Default, inputType=FunctionInput_MaterialAttributes}
  n2 = FunctionInput() {inputName=Low/Mobile, inputType=FunctionInput_MaterialAttributes, sortPriority=1}
  FunctionOutput(None=ShadingPathSwitch(Default=QualitySwitch(Default=@n1, Low=@n2), Mobile=@n2))
 ]
}
