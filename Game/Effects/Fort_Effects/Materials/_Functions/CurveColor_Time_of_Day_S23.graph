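# Material function that returns the sky color for the current time of day using the Season 23 color curve, exposed to the material library.
# It feeds the global TimeOfDay_0-1 collection parameter in as curve time to a curve atlas row parameter reading Curve_TimeOfDay_S23 from AthenaCurveAtlas01.
# Its description notes this is more accurate after S11 than the four-color time-of-day function.
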MaterialFunction /Game/Effects/Fort_Effects/Materials/_Functions/CurveColor_Time_of_Day_S23
graph {
 properties {
  bExposeToLibrary=true
  description="Using the time of day to sample color on a curve gives you the sky Color. After S11, this color is more accurate than the 4 color time of day function."
 }
 nodes [
  n1 = CurveAtlasRowParameter(CurveTime=CollectionParameter() {collection=&FortniteMaterialParameters, parameterName=TimeOfDay_0-1}) {
   atlas=&AthenaCurveAtlas01
   curve=/Game/Athena/LinearColorCurves/Curve_TimeOfDay_S23
   defaultValue=321
   group=TimeOfDay
   parameterName=TOD_Curve
  }
  FunctionOutput(None=@n1)
 ]
}
