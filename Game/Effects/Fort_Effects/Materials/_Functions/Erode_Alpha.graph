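# Erode_Alpha is a material function in the Misc and FX_Common library categories that erodes a single-channel alpha using a dynamic parameter.
# It offsets the parameter by two constants to form lower and upper bounds, then lerps between them using the input alpha.
# This stretches the alpha's contrast and shifts it with the parameter.
# The output is unclamped, so it can fall outside the 0–1 range.
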MaterialFunction /Game/Effects/Fort_Effects/Materials/_Functions/Erode_Alpha
graph {
 properties {bExposeToLibrary=true, description=Erode Alpha with a Dynamic Parameter, libraryCategoriesText [Misc,FX_Common]}
 nodes [
  n1 = Constant() {r=-0.5}
  n2 = FunctionInput() {description=Input Dynamic Param to drive Alpha, inputName=Dynamic Param, inputType=FunctionInput_Scalar}
  n3 = Constant() {r=1}
  FunctionOutput(None=LinearInterpolate(
    A=Add(A=@n1, B=@n2)
    B=Add(A=@n3, B=@n2)
    Alpha=FunctionInput() {description=Input Single Channel Alpha, inputName=Alpha, inputType=FunctionInput_Scalar}
   ))
 ]
}
