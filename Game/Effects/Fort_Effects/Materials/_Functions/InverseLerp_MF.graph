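# InverseLerp_MF is a material function, exposed to the library under Math, that computes where a scalar value lies between a min and a max.
# It subtracts min from value and divides by max minus min, with min defaulting to 0, max to 1 and value to 0.
# It outputs both a saturated Clamped result and the raw Unclamped ratio, with no guard against max equaling min.
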MaterialFunction /Game/Effects/Fort_Effects/Materials/_Functions/InverseLerp_MF
graph {
 properties {bExposeToLibrary=true, libraryCategoriesText [Math]}
 nodes [
  n1 = Constant()
  n2 = FunctionInput(Preview=Constant()) {bUsePreviewValueAsDefault=true, inputName=min, inputType=FunctionInput_Scalar}
  n3 = Constant() {r=1}
  n4 = Divide(
   A=Subtract(A=FunctionInput(Preview=@n1) {bUsePreviewValueAsDefault=true, inputName=value, inputType=FunctionInput_Scalar, sortPriority=2}, B=@n2)
   B=Subtract(A=FunctionInput(Preview=@n3) {bUsePreviewValueAsDefault=true, inputName=max, inputType=FunctionInput_Scalar, sortPriority=1}, B=@n2)
  )
  FunctionOutput(None=Saturate(None=@n4)) {outputName=Clamped}
  FunctionOutput(None=@n4) {outputName=Unclamped, sortPriority=1}
 ]
}
