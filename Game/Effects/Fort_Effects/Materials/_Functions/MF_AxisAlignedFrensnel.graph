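# Library material function (FX Common) that computes a fresnel aligned to an axis, by default the mesh's local Z for vertical cylinders.
# Double cross products project the camera vector perpendicular to the axis; its dot with the normal (default vertex normal) is clamped and powered, peaking where facing camera.
# A second output divides that dot by the projected-to-view vector dot, estimating ray travel through cylinder sides, ignoring end caps.
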MaterialFunction /Game/Effects/Fort_Effects/Materials/_Functions/MF_AxisAlignedFrensnel
graph {
 properties {bExposeToLibrary=true, description=Axis aligned fresnel, libraryCategoriesText [FX Common]}
 nodes [
  n1 = Normalize(VectorInput=Subtract(A=WorldPosition().XYZ, B=CameraPositionWS()))
  n2 = FunctionInput(Preview=Transform(None=Constant3Vector() {constant=<rgba 0 0 1 0>}) {transformSourceType=TRANSFORMSOURCE_Local}) {
   bUsePreviewValueAsDefault=true
   description="The axis to align to. By default this will be the local Z vector of the mesh which means it works with the vertically oriented Engine\\Content cylinder meshes."
   inputName=Axis
  }
  n3 = CrossProduct(A=Normalize(VectorInput=CrossProduct(A=@n1, B=@n2)), B=@n2)
  n4 = Normalize(VectorInput=@n3)
  n5 = FunctionInput(Preview=VertexNormalWS()) {
   bUsePreviewValueAsDefault=true
   description=Defaults to Vertex Normal but this allows you to specify a custom normal. Normal should be in world space.
   inputName=Normal
   sortPriority=1
  }
  n6 = DotProduct(A=@n4, B=@n5)
  n7 = FunctionInput() {
   bUsePreviewValueAsDefault=true
   description=The exponent will be applied as a Power to the result.
   inputName=Exponent
   inputType=FunctionInput_Scalar
   sortPriority=1
   previewValue=<1 0 0 1>
  }
  FunctionOutput(None=Power(Base=Max(A=@n6) {constB=0}, Exp=@n7)) {description=The resulting fresnel with Power Exponent applied., outputName=Fresnel}
  n8 = Normalize(VectorInput=Subtract(A=WorldPosition().XYZ, B=CameraPositionWS()))
  FunctionOutput(None=Divide(A=@n6, B=DotProduct(A=@n4, B=@n8))) {
   description="This returns the value for how far a ray would travel through a cylinder, where a value of 1 equals 1 diameter of the cylinder. It does not handle end caps, only the sides of the cylinder."
   outputName=Cylinder Thickness
   sortPriority=1
  }
  Multiply(A=@n3, B=Constant3Vector() {constant=<rgba -1 -1 -1 0>})
  Multiply(A=@n6) {constB=-1}
 ]
}
