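# Material function of shared building blocks for blaster laser particle materials.
# ElementColor indexes a curve atlas color ramp, with a small offset separating bright from dark colors.
# A dynamic-parameter laser distance scales a V-axis edge mask; particle relative time and HeadSpeed scroll a ramp curve along V for two head-enhance outputs.
# It also outputs scaled time, a looping preview pulse and a global WPO scalar.
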MaterialFunction /Game/Effects/Fort_Effects/Materials/_Functions/MF_BlasterCommon
graph {nodes [
  Constant()
  n1 = Time()
  n2 = OneMinus(None=Frac(None=@n1))
  n3 = Sine(None=Multiply(A=Power(Base=@n2) {constExponent=10}) {constB=0.5})
  FunctionOutput(None=OneMinus(None=Saturate(None=@n3))) {outputName=PreviewTime}
  n4 = FunctionInput(Preview=ScalarParameter() {parameterName=ElementColor}) {bUsePreviewValueAsDefault=true, inputName=ElementColor, inputType=FunctionInput_Scalar, sortPriority=1}
  n5 = CurveAtlasRowParameter(CurveTime=Add(A=@n4) {constB=0.05}) {atlas=&AthenaCurveAtlas01, curve=/Game/Athena/LinearColorCurves/Balster/Curve_ElementColors, defaultValue=59}
  FunctionOutput(None=@n5) {outputName=BrightColor, sortPriority=2}
  n6 = CurveAtlasRowParameter(CurveTime=@n4) {atlas=&AthenaCurveAtlas01, curve=&Curve_ElementColors, defaultValue=59}
  FunctionOutput(None=@n6) {outputName=DarkColor, sortPriority=1}
  n7 = Constant() {r=0.75}
  FunctionOutput(None=Multiply(A=@n1, B=FunctionInput(Preview=@n7) {bUsePreviewValueAsDefault=true, inputName=TimeScale, inputType=FunctionInput_Scalar}) {constB=0.5}) {outputName=ScaledTime}
  n8 = Subtract(A=ComponentMask(None=TextureCoordinate()) {g=true}) {constB=0.5}
  n9 = OneMinus(None=Multiply(A=Abs(None=@n8)) {constB=2})
  n10 = Divide(
   A=DynamicParameter() {defaultValue=<rgba 1 1 1 10000>, paramNames [Param1, Param2, Param3, BP_LaserDistance]}.BP_LaserDistance
   B=Constant() {r=200}
  )
  FunctionOutput(None=Saturate(None=Multiply(A=@n9, B=@n10) {constB=10})) {outputName=EdgeMask, sortPriority=3}
  n11 = FunctionInput(Preview=Constant() {r=1}) {bUsePreviewValueAsDefault=true, inputName=HeadEnhanceScale, inputType=FunctionInput_Scalar, sortPriority=3, previewValue=<1 0 0 1>}
  n12 = Saturate(None=Multiply(A=@n11, B=ParticleRelativeTime()))
  n13 = Saturate(None=Multiply(A=Frac(None=@n12), B=ScalarParameter() {defaultValue=4, parameterName=HeadSpeed}))
  n14 = Subtract(A=OneMinus(None=TextureCoordinate()))
  n15 = CurveAtlasRowParameter(CurveTime=Add(A=Multiply(A=@n13) {constB=2}, B=ComponentMask(None=@n14) {g=true})) {
   atlas=&AthenaCurveAtlas01
   curve=/Game/Athena/LinearColorCurves/Balster/Curve_LaserEffectRamp
   defaultValue=0.2285
   parameterName=CurvePara1
  }
  FunctionOutput(None=@n15.R) {outputName=HeadEnhance, sortPriority=4}
  FunctionOutput(None=@n15.G) {outputName=HeadEnhance1, sortPriority=5}
  FunctionOutput(None=ScalarParameter() {defaultValue=0.25, parameterName=GloabalWpo}) {outputName=GlobalWPO, sortPriority=6}
  Multiply(A=@n14) {constB=0.5}
  ScalarParameter() {defaultValue=1000, parameterName=BP_LaserDistance}
  Time()
 ]}
