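# MF_ClipAlpha is a library-exposed material function in the effects folder that erodes an alpha texture as an AlphaOverTime value falls.
# The inverted AlphaOverTime moves a threshold from negative ClipRange up to one, which is subtracted from AlphaTexture.
# Dividing by ClipRange and saturating gives a soft edge of that width.
# The offset start makes alpha fully opaque at one and fully clear at zero, assuming texture values within zero to one.
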MaterialFunction /Game/Effects/Fort_Effects/Materials/_Functions/MF_ClipAlpha
graph {
 properties {bExposeToLibrary=true}
 nodes [
  n1 = FunctionInput(Preview=Constant() {r=0.1}) {inputName=ClipRange, inputType=FunctionInput_Scalar}
  n2 = FunctionInput(Preview=Constant() {r=1}) {inputName=AlphaOverTime, inputType=FunctionInput_Scalar}
  n3 = Subtract(
   A=FunctionInput(Preview=Constant() {r=1}) {inputName=AlphaTexture, inputType=FunctionInput_Scalar}
   B=LinearInterpolate(A=Multiply(A=@n1) {constB=-1}, Alpha=OneMinus(None=@n2) {desc=Inverse Alpha})
  )
  FunctionOutput(None=Saturate(None=Divide(A=@n3, B=@n1)))
 ]
}
