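# Material function that warps decal UVs into sphere-like coordinates, scaling them around their center.
# The blue component of LightVector, likely serving as decal-space position, feeds a soft sphere mask centered at 0.5 to measure distance from the middle.
# A one-minus, power and square-root chain turns that distance into a circular falloff.
# The falloff interpolates the scale fed to ScaleUVsByCenter, with an Additional Scale input adjusting the upper end.
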MaterialFunction /Game/Effects/Fort_Effects/Materials/_Functions/MF_DecalSphereCoords
graph {nodes [
  n1 = TextureCoordinate()
  n2 = Add(A=FunctionInput() {bUsePreviewValueAsDefault=true, inputName=Additional Scale, inputType=FunctionInput_Scalar})
  n3 = LightVector()
  n4 = OneMinus(None=SphereMask(
    A=MaterialFunctionCall(Float3=@n3) {materialFunction=&BreakOutFloat3Components}.B
    B=Constant() {r=0.5}
   ) {attenuationRadius=0.5, hardnessPercent=0})
  n5 = SquareRoot(None=OneMinus(None=Power(Base=@n4)))
  FunctionOutput(None=MaterialFunctionCall(
    UVs=FunctionInput(Preview=@n1) {bUsePreviewValueAsDefault=true, inputName=UVs, inputType=FunctionInput_Vector2}
    Texture Scale=LinearInterpolate(B=@n2, Alpha=@n5) {constB=2}
   ) {materialFunction=&ScaleUVsByCenter}.UVs)
 ]}
