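# MF_Dissolve is an effects material function that turns a scalar ramp into an animatable dissolve mask with adjustable edge sharpness.
# It scales Ramp by EdgeSharpness and offsets it by Dissolve, remapped to account for that sharpness.
# It outputs the result clamped and unclamped, plus an Edge band made by feeding the scaled, clamped result through a sine.
# Dissolve defaults to a debug time sine, Ramp to texture coordinates.
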MaterialFunction /Game/Effects/Fort_Effects/Materials/_Functions/MF_Dissolve
graph {nodes [
  n1 = TextureCoordinate()
  n2 = FunctionInput(Preview=Constant() {r=20}) {bUsePreviewValueAsDefault=true, inputName=EdgeSharpness, inputType=FunctionInput_Scalar, sortPriority=3}
  n3 = FunctionInput(Preview=MaterialFunctionCall() {materialFunction=&DebugTimeSine}.Result) {bUsePreviewValueAsDefault=true, inputName=Dissolve, inputType=FunctionInput_Scalar, sortPriority=2}
  n4 = Add(A=@n2)
  n5 = Subtract(
   A=Multiply(A=FunctionInput(Preview=@n1) {bUsePreviewValueAsDefault=true, inputName=Ramp, inputType=FunctionInput_Scalar, sortPriority=1}, B=@n2)
   B=Subtract(A=Multiply(A=@n3, B=@n4))
  )
  FunctionOutput(None=Clamp(None=@n5)) {outputName=ClampedResult}
  FunctionOutput(None=@n5) {outputName=ResultWithOutClamp}
  n6 = Divide(A=@n5) {constB=4}
  FunctionOutput(None=Sine(None=Clamp(None=@n6) {maxDefault=0.5})) {outputName=Edge, sortPriority=3}
 ]}
