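# Material function exposed to the VFX library that measures how far scene geometry lies behind the current translucent pixel, as a scaled gradient.
# It subtracts pixel depth plus a bias from scene depth, then divides by a divisor, with overridable inputs defaulting to depth nodes and parameters.
# A "Normalized" variant multiplies the result by absolute vertex normal dot camera vector, scaling it by viewing angle.
# Both variants output clamped, unclamped and inverted versions.
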MaterialFunction /Game/Effects/Fort_Effects/Materials/_Functions/MF_DistanceBehindTranslucency_01
graph {
 properties {bExposeToLibrary=true, libraryCategoriesText [VFX]}
 nodes [
  n1 = SceneDepth()
  n2 = FunctionInput(Preview=PixelDepth().R) {bUsePreviewValueAsDefault=true, inputName=Pixel Depth, inputType=FunctionInput_Scalar, sortPriority=3}
  n3 = FunctionInput(Preview=ScalarParameter() {parameterName=Distance Bias}) {bUsePreviewValueAsDefault=true, inputName=Distance Bias, inputType=FunctionInput_Scalar}
  n4 = Divide(
   A=Subtract(
    A=FunctionInput(Preview=@n1.R) {bUsePreviewValueAsDefault=true, inputName=Scene Depth, inputType=FunctionInput_Scalar, sortPriority=2}
    B=Add(A=@n2, B=@n3)
   )
   B=FunctionInput(Preview=ScalarParameter() {defaultValue=3000, parameterName=Distance Divisor}) {bUsePreviewValueAsDefault=true, inputName=Distance Divisor, inputType=FunctionInput_Scalar, sortPriority=1}
  )
  FunctionOutput(None=OneMinus(None=@n4)) {outputName=Inverted Unclamped, sortPriority=3}
  n5 = VertexNormalWS()
  n6 = CameraVectorWS()
  n7 = Multiply(A=Abs(None=DotProduct(A=@n5, B=@n6)), B=@n4)
  n8 = Saturate(None=@n7)
  FunctionOutput(None=@n8) {outputName=Normalized, sortPriority=4}
  FunctionOutput(None=@n7) {outputName=Normalized Unclamped, sortPriority=5}
  FunctionOutput(None=OneMinus(None=@n8)) {outputName=Normalized Inverted, sortPriority=6}
  FunctionOutput(None=OneMinus(None=@n7)) {outputName=Normalized Inverted Unclamped, sortPriority=7}
  FunctionOutput(None=@n4) {outputName=Unclamped, sortPriority=1}
  n9 = Saturate(None=@n4)
  FunctionOutput(None=@n9) {outputName=Clamped}
  FunctionOutput(None=OneMinus(None=@n9)) {outputName=Inverted, sortPriority=2}
 ]
}
