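# VFX material function that fades an effect by camera distance, by default fading it out as the camera approaches.
# It measures from object position, overridable and offsettable, so by default the whole object fades uniformly rather than per pixel.
# The distance is clamped between fade end and start, normalized, then multiplied into an optional opacity.
# Outputs are the faded result, an inverted version needing flipped start/end inputs, and the raw distance.
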MaterialFunction /Game/Effects/Fort_Effects/Materials/_Functions/MF_Distance_SyncedFade_FX
graph {
 properties {bExposeToLibrary=true, libraryCategoriesText [Misc,VFX]}
 nodes [
  n1 = FunctionInput(Preview=ObjectPositionWS()) {
   bUsePreviewValueAsDefault=true
   description="If you want to override what the camera position is measured against (ie. object position, particle position, etc), input here."
   inputName=Position
   sortPriority=1
  }
  n2 = FunctionInput() {
   bUsePreviewValueAsDefault=true
   description="Allows you to manually offset the pivot (in world space) if desired. Useful if you have particle systems that are stretched execisvely in one axis."
   inputName=Pivot_Manual_Offset
   sortPriority=4
  }
  n3 = Distance(A=Add(A=@n1, B=@n2), B=CameraPositionWS())
  FunctionOutput(None=@n3) {description=If you want to sample the distance value this function produces, outputName=Distance, sortPriority=2}
  n4 = FunctionInput() {
   bUsePreviewValueAsDefault=true
   description="Plug your opacity in here for a multiply against the distance math in this function. Doesn't require input, just added for convenience."
   inputName=Opacity In
   inputType=FunctionInput_Scalar
   previewValue=<1 1 1 1>
  }
  n5 = FunctionInput() {
   bUsePreviewValueAsDefault=true
   description=The distance that the emitter should be completely faded away
   inputName=Fade_End_Distance
   inputType=FunctionInput_Scalar
   sortPriority=3
   previewValue=<100 0 0 1>
  }
  n6 = FunctionInput() {
   bUsePreviewValueAsDefault=true
   description=The distance that the emitter should start to fade away at
   inputName=Fade_Start_Distance
   inputType=FunctionInput_Scalar
   sortPriority=2
   previewValue=<500 0 0 1>
  }
  n7 = Subtract(A=Clamp(None=@n3, Min=@n5, Max=@n6), B=@n5) {desc=Makes range start at 0}
  n8 = Subtract(A=@n6, B=@n5) {desc=This establishes the fading range}
  n9 = Saturate(None=Abs(None=Divide(A=@n7, B=@n8)))
  FunctionOutput(None=Multiply(A=@n4, B=@n9))
  FunctionOutput(None=Multiply(A=@n4, B=OneMinus(None=@n9))) {description="Note! When using inverted output, the start/end inputs must be flipped", outputName=Result Inverted, sortPriority=1}
 ]
}
