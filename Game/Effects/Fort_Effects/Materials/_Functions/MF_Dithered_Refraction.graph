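# Material function exposed to the library that dithers a scalar refractive index for refraction effects.
# It derives a spread as half the index's absolute distance from 1, then forms multipliers of one plus and one minus that spread.
# DitherTemporalAA blends between the two multipliers, and the result scales the input index, jittering refraction strength per pixel and over time in proportion to its deviation from 1.
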MaterialFunction /Game/Effects/Fort_Effects/Materials/_Functions/MF_Dithered_Refraction
graph {
 properties {bExposeToLibrary=true}
 nodes [
  n1 = FunctionInput() {inputName=RefractiveIndex, inputType=FunctionInput_Scalar}
  n2 = Abs(None=Divide(A=Subtract(A=@n1)))
  n3 = Add(A=@n2)
  n4 = Subtract(B=@n2)
  n5 = MaterialFunctionCall() {materialFunction=&DitherTemporalAA}
  FunctionOutput(None=Multiply(A=@n1, B=LinearInterpolate(A=@n3, B=@n4, Alpha=@n5.Result) {constA=0.9, constB=1.1}))
 ]
}
