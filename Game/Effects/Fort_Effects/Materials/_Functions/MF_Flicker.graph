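# Material function that generates a procedural flicker signal over time for effects materials.
# Time is offset by a FlickerRandom input to vary the phase, then scaled by FlickerInitialSpeed.
# The result drives four sine waves with different periods, which are multiplied together into an irregular pulse.
# The output is the raw signed product, with no clamp or remap to a 0 to 1 range.
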MaterialFunction /Game/Effects/Fort_Effects/Materials/_Functions/MF_Flicker
graph {nodes [
  n1 = Constant()
  n2 = Multiply(
   A=Add(A=Time(), B=FunctionInput(Preview=@n1) {inputName=FlickerRandom, inputType=FunctionInput_Scalar, sortPriority=1})
   B=FunctionInput(Preview=Constant() {r=4}) {inputName=FlickerInitialSpeed, inputType=FunctionInput_Scalar}
  ) {constB=80}
  n3 = Multiply(A=Sine(None=@n2) {period=0.155}, B=Sine(None=@n2) {period=0.2})
  n4 = Sine(None=@n2) {period=0.43}
  FunctionOutput(None=Multiply(A=Sine(None=@n2) {period=0.075}, B=Multiply(A=@n3, B=@n4)))
 ]}
