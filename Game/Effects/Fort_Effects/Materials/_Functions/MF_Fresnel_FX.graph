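# Library-exposed material function for effects materials that computes several Fresnel-style view-angle masks.
# It dots a camera vector with a normal, defaulting to the world camera vector and vertex normal, with Brightness and Exponent shaping the result.
# Plain outputs brighten facing surfaces and inverted ones edges, in clamped, unclamped and abs-based double-sided versions.
# Brightness multiplies before the power in single-sided outputs but after it in double-sided ones.
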MaterialFunction /Game/Effects/Fort_Effects/Materials/_Functions/MF_Fresnel_FX
graph {
 properties {bExposeToLibrary=true}
 nodes [
  n1 = FunctionInput(Preview=CameraVectorWS()) {
   bUsePreviewValueAsDefault=true
   description=Allows manual control of the camera vector
   inputName=Camera Vector
   sortPriority=2
   previewValue=<1 1 1 1>
  }
  n2 = FunctionInput(Preview=VertexNormalWS()) {bUsePreviewValueAsDefault=true, inputName=Vertex / Pixel Normal, sortPriority=3, previewValue=<1 1 1 1>}
  n3 = DotProduct(A=@n1, B=@n2)
  n4 = Abs(None=@n3)
  n5 = OneMinus(None=@n4)
  n6 = FunctionInput() {bUsePreviewValueAsDefault=true, inputName=Exponent, inputType=FunctionInput_Scalar, sortPriority=1, previewValue=<1 0 0 1>}
  n7 = FunctionInput() {bUsePreviewValueAsDefault=true, inputName=Brightness, inputType=FunctionInput_Scalar, previewValue=<1 0 0 1>}
  FunctionOutput(None=Multiply(A=@n7, B=Power(Base=@n5, Exp=@n6))) {outputName=Fresnel Inverted Double Sided, sortPriority=5}
  n8 = Clamp(None=@n3)
  n9 = Power(Base=Multiply(A=@n8, B=@n7), Exp=@n6)
  FunctionOutput(None=Clamp(None=@n9)) {outputName=Fresnel}
  FunctionOutput(None=@n9) {outputName=Fresnel Unclamped, sortPriority=2}
  n10 = Power(Base=Multiply(A=@n7, B=OneMinus(None=@n8)), Exp=@n6)
  FunctionOutput(None=Clamp(None=@n10)) {outputName=Fresnel Inverted, sortPriority=1}
  FunctionOutput(None=@n10) {outputName=Fresnel Inverted Unclamped, sortPriority=3}
  FunctionOutput(None=Multiply(A=@n7, B=Power(Base=@n4, Exp=@n6))) {outputName=Fresnel Double Sided, sortPriority=4}
 ]
}
