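# Material function that outputs an animated scalar for irregular, lightning-like flashes, driven by Time scaled by a TimeMult input.
# Two slow sines, scaled just above one by Frequency1 and Frequency2 parameters, are floored, summed and clamped into a stepped gate that opens only briefly near wave peaks.
# Three faster sines combine into a bias-scaled flicker that multiplies the gate before a final clamp.
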MaterialFunction /Game/Effects/Fort_Effects/Materials/_Functions/MF_Lightning_Flashes
graph {nodes [
  n1 = Multiply(A=Time(), B=FunctionInput(Preview=Constant() {r=1}) {bUsePreviewValueAsDefault=true, inputName=TimeMult, inputType=FunctionInput_Scalar})
  n2 = Multiply(A=Sine(None=@n1) {period=1.27}, B=ScalarParameter() {defaultValue=1.05, parameterName=Frequency1}) {constB=1.05}
  n3 = Multiply(A=Sine(None=@n1) {period=2.5}, B=ScalarParameter() {defaultValue=1.07, parameterName=Frequency2}) {constB=1.07}
  n4 = Clamp(None=Add(A=Floor(None=@n2), B=Floor(None=@n3))) {maxDefault=2}
  n5 = Sine(None=@n1) {period=0.19}
  n6 = Sine(None=@n1) {period=0.115}
  n7 = ConstantBiasScale(None=Add(A=Multiply(A=@n5, B=@n6), B=Sine(None=@n1) {period=0.07}))
  FunctionOutput(None=Clamp(None=Multiply(A=@n4, B=@n7)) {maxDefault=5})
 ]}
