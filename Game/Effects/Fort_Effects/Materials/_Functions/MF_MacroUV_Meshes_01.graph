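# Material function that generates macro UVs for meshes by projecting each pixel onto a virtual plane, by default camera-facing at the actor position.
# It intersects the camera-to-pixel ray with the plane defined by Plane Center and Plane Normal Axis.
# The hit offset is dotted with two axes from CreateThirdOrthogonalVector to form 2D coordinates, then divided by Scale.
# Intersections at or behind the camera output zero.
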MaterialFunction /Game/Effects/Fort_Effects/Materials/_Functions/MF_MacroUV_Meshes_01
graph {
 properties {bExposeToLibrary=true}
 nodes [
  n1 = FunctionInput(Preview=CameraVectorWS()) {bUsePreviewValueAsDefault=true, inputName=Plane Normal Axis, previewValue=<0 0 1 1>}
  n2 = DotProduct(A=@n1, B=CameraPositionWS() {desc=Ray Origin})
  n3 = FunctionInput(Preview=ActorPositionWS()) {bUsePreviewValueAsDefault=true, desc=Plane Origin, description=The center location of the virtual Plane, inputName=Plane Center}
  n4 = DotProduct(A=@n1, B=@n3) {desc=Distance to Plane}
  n5 = Subtract(A=WorldPosition().XYZ, B=CameraPositionWS()) {desc=Direction Vector}
  n6 = Divide(A=Multiply(A=Subtract(A=@n2, B=@n4) {desc=Plane Distance}) {constB=-1}, B=DotProduct(A=@n1, B=@n5) {desc=VectorDotPlaneNormal}) {desc=Intersection Time1}
  n7 = Constant()
  n8 = MaterialFunctionCall!(Vector1=@n1, Vector2=Constant3Vector() {constant=<rgba 0 0.0002 1 0>}) {materialFunction=/Engine/Functions/Engine_MaterialFunctions02/Math/CreateThirdOrthogonalVector}
  n9 = Subtract(A=Add(A=CameraPositionWS() {desc=Ray Origin}, B=Multiply(A=@n6, B=@n5)) {desc=Intersection Position}, B=@n3)
  n10 = AppendVector(A=DotProduct(A=Multiply(A=@n8.Vector3) {desc=Plane X}, B=@n9), B=DotProduct(A=Multiply(A=@n8.Vector2) {desc=Plane Y}, B=@n9))
  n11 = Constant()
  n12 = FunctionInput() {bUsePreviewValueAsDefault=true, inputName=Scale, inputType=FunctionInput_Scalar, sortPriority=10, previewValue=<1 0 0 1>}
  FunctionOutput(None=Divide(A=If(A=@n6, B=@n7, "A > B"=@n10, "A == B"=@n11, "A < B"=@n11), B=@n12))
 ]
}
