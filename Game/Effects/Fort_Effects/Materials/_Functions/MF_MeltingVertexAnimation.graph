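# Material function that produces a melting vertex wobble as world position offset, plus the sampled normal as a second output.
# A world-aligned normal texture is sampled using particle- or actor-relative position, scrolled over time along Z, with a world-XY term that pans it during movement, zeroed on mobile and lower quality.
# Its XY projection forms the offset vector, masked by relative height to fade toward the top and scaled by intensity.
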MaterialFunction /Game/Effects/Fort_Effects/Materials/_Functions/MF_MeltingVertexAnimation
graph {nodes [
  n1 = TextureObjectParameter() {
   group=VertexAnim
   parameterName=VertexAnimNormalMain
   samplerType=SAMPLERTYPE_Normal
   texture=/Game/Effects/Fort_Effects/Textures/T_DustDistortion
  }
  n2 = WorldPosition()
  n3 = Multiply(
   A=AppendVector(A=ComponentMask(None=@n2.XYZ) {g=true, r=true}, B=Constant() {r=0.2})
   B=FunctionInput(Preview=ScalarParameter() {defaultValue=1, group=VertexAnim, parameterName=WPO_PanSpeedWhileMoving}) {inputName=WPO_PanSpeedWhileMoving, inputType=FunctionInput_Scalar, sortPriority=3}
  ) {constB=0.1}
  n4 = Constant()
  n5 = ShadingPathSwitch(Default=@n3, Deferred=@n3, Mobile=@n4) {desc=Migrated from FeatureLevelSwitch}
  n6 = WorldPosition()
  n7 = FunctionInput(Preview=StaticBool()) {inputName=UseParticlePositionOrObjectPosition, inputType=FunctionInput_StaticBool, sortPriority=5}
  n8 = ActorPositionWS()
  n9 = Subtract(
   A=StaticSwitch(True=@n6.XYZ, False=TransformPosition(None=@n6.XYZ) {transformSourceType=TRANSFORMPOSSOURCE_World}, Value=@n7)
   B=StaticSwitch(True=ParticlePositionWS(), False=TransformPosition(None=@n8) {transformSourceType=TRANSFORMPOSSOURCE_World}, Value=@n7)
  )
  n10 = Constant()
  n11 = ScalarParameter() {defaultValue=30, group=VertexAnim, parameterName=VertexAnimNormalSpeed}
  n12 = MaterialFunctionCall(
   X=@n10
   Y=@n10
   Z=Multiply(A=Time(), B=FunctionInput(Preview=@n11) {inputName=VertexAnimNormalSpeed, inputType=FunctionInput_Scalar, sortPriority=2}) {constB=30}
  ) {materialFunction=&MakeFloat3}
  n13 = MaterialFunctionCall(
   TextureObject=FunctionInput(Preview=@n1) {inputName=VertexAnimNormalMain, inputType=FunctionInput_Texture2D}
   TextureSize=FunctionInput(Preview=ScalarParameter() {defaultValue=75, group=VertexAnim, parameterName=VertexAnimTexSize}) {inputName=VertexAnimTexSize, inputType=FunctionInput_Scalar, sortPriority=1}
   WorldPosition=Add(A=QualitySwitch(Default=@n5, Low=@n4, High=@n5, Medium=@n4), B=Add(A=@n9, B=@n12.Result))
  ) {materialFunction=&WorldAlignedTexture}
  n14 = Divide(
   A=@n13.XY Texture
   B=FunctionInput(Preview=ScalarParameter() {defaultValue=0.3, group=VertexAnim, parameterName=VertexAnimIntesityXY}) {inputName=VertexAnimIntesityXY, inputType=FunctionInput_Scalar, sortPriority=6}
  ) {constB=3}
  n15 = MaterialFunctionCall(X=ComponentMask(None=@n14) {r=true}, Y=ComponentMask(None=@n14) {g=true}, Z=Multiply(A=ComponentMask(None=@n13.XY Texture) {g=true})) {materialFunction=&MakeFloat3}
  n16 = WorldPosition()
  n17 = ActorPositionWS()
  n18 = Subtract(
   A=StaticSwitch(True=WorldPosition().XYZ, False=TransformPosition(None=@n16.XYZ) {transformSourceType=TRANSFORMPOSSOURCE_World}, Value=@n7)
   B=StaticSwitch(True=ParticlePositionWS(), False=TransformPosition(None=@n17) {transformSourceType=TRANSFORMPOSSOURCE_World}, Value=@n7)
  )
  n19 = Add(
   A=Divide(A=ComponentMask(None=@n18) {b=true}) {constB=180}
   B=FunctionInput(Preview=ScalarParameter() {defaultValue=0.3, group=VertexAnim, parameterName=VertexAnimVerticalMaskBias}) {inputName=VertexAnimVerticalMaskBias, inputType=FunctionInput_Scalar, sortPriority=8}
  ) {constB=0.3}
  n20 = Saturate(None=OneMinus(None=Multiply(A=@n19) {constB=2}))
  n21 = FunctionInput(Preview=ScalarParameter() {defaultValue=0.3, group=VertexAnim, parameterName=WPO_Intensity}) {inputName=WPO_Intensity, inputType=FunctionInput_Scalar, sortPriority=7}
  FunctionOutput(None=Multiply(A=LinearInterpolate(B=@n15.Result, Alpha=@n20) {constB=0}, B=LinearInterpolate(Alpha=@n21) {constB=100})) {outputName="VertexAnimation(WPO)"}
  FunctionOutput(None=@n13.XYZ Texture) {outputName="XYZ(Normal)", sortPriority=1}
  ObjectPositionWS()
  ObjectPositionWS()
 ]}
