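# MF_ModuloTime is a library-exposed material function that turns time into a repeating ramp for effect materials.
# It divides the Time input, which defaults to the material Time node, by Divisor to set the rate, then applies Fmod with Mod to wrap it.
# It outputs the wrapped value both unclamped and saturated to 0-1.
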MaterialFunction /Game/Effects/Fort_Effects/Materials/_Functions/MF_ModuloTime
graph {
 properties {bExposeToLibrary=true}
 nodes [
  n1 = FunctionInput(Preview=Time()) {bUsePreviewValueAsDefault=true, inputName=Time, inputType=FunctionInput_Scalar, previewValue=<1 0 0 0>}
  n2 = FunctionInput() {bUsePreviewValueAsDefault=true, inputName=Divisor, inputType=FunctionInput_Scalar, sortPriority=1, previewValue=<1 0 0 0>}
  n3 = FunctionInput() {bUsePreviewValueAsDefault=true, inputName=Mod, inputType=FunctionInput_Scalar, sortPriority=2, previewValue=<1 0 0 0>}
  n4 = Fmod(A=Divide(A=@n1, B=@n2), B=@n3)
  FunctionOutput(None=@n4) {outputName=Unclamped}
  FunctionOutput(None=Saturate(None=@n4)) {outputName=Clamped 0-1, sortPriority=1}
 ]
}
