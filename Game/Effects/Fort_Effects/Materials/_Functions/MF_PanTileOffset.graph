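# Material function that provides panning, tiling and offset of UV coordinates for effects materials, exposed to the function library.
# It multiplies the input UV by the per-axis tile values, adds per-axis pan speed multiplied by time, then adds the offset.
# All inputs have preview defaults: texture coordinates for UV, unit tiling, zero pan and offset, and a Time node with an overridden long looping period.
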MaterialFunction /Game/Effects/Fort_Effects/Materials/_Functions/MF_PanTileOffset
graph {
 properties {bExposeToLibrary=true, description="Provides panning, tiling, and offset"}
 nodes [
  n1 = FunctionInput() {bUsePreviewValueAsDefault=true, inputName=Offset U, inputType=FunctionInput_Scalar, sortPriority=5}
  n2 = FunctionInput() {bUsePreviewValueAsDefault=true, inputName=Offset V, inputType=FunctionInput_Scalar, sortPriority=6}
  n3 = FunctionInput() {bUsePreviewValueAsDefault=true, inputName=Tile U, inputType=FunctionInput_Scalar, sortPriority=1, previewValue=<1 0 0 1>}
  n4 = FunctionInput() {bUsePreviewValueAsDefault=true, inputName=Tile V, inputType=FunctionInput_Scalar, sortPriority=2, previewValue=<1 0 0 1>}
  n5 = Multiply(
   A=FunctionInput(Preview=TextureCoordinate()) {bUsePreviewValueAsDefault=true, inputName=UV, inputType=FunctionInput_Vector2}
   B=AppendVector(A=@n3, B=@n4)
  )
  n6 = Multiply(
   A=AppendVector(
    A=FunctionInput() {bUsePreviewValueAsDefault=true, inputName=Pan U, inputType=FunctionInput_Scalar, sortPriority=3}
    B=FunctionInput() {bUsePreviewValueAsDefault=true, inputName=Pan V, inputType=FunctionInput_Scalar, sortPriority=4}
   )
   B=FunctionInput(Preview=Time() {bOverride_Period=true, period=8000}) {bUsePreviewValueAsDefault=true, inputName=Time, inputType=FunctionInput_Scalar, sortPriority=7}
  )
  FunctionOutput(None=Add(A=AppendVector(A=@n1, B=@n2), B=Add(A=@n5, B=@n6))) {description="Provides panning, tiling, and offset", outputName=UV's}
 ]
}
