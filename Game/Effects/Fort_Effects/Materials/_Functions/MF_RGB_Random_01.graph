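# Material function for VFX that uses a random scalar to pick one channel of a channel-packed texture.
# The picker's fractional part drives nested Ifs selecting Option 3 from 0.66 upward, otherwise Option 2 or 1 through an inner If whose B is unconnected.
# Options default to pure red, green and blue masks but accept custom vectors.
# The pick is dotted with the texture input, and also output alone as a vector.
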MaterialFunction /Game/Effects/Fort_Effects/Materials/_Functions/MF_RGB_Random_01
graph {
 properties {
  bExposeToLibrary=true
  description="If you are channel packing a texture and want a dynamic parameter (or other random scalar) to pick between RGB individually, this will do this quickly for you."
  libraryCategoriesText [Misc,VFX]
 }
 nodes [
  n1 = FunctionInput() {
   bUsePreviewValueAsDefault=true
   desc="Your random range (usually a dynamic paramater in 0-1 space)"
   inputName=Random Picker
   inputType=FunctionInput_Scalar
  }
  n2 = Frac(None=@n1)
  n3 = FunctionInput() {
   bUsePreviewValueAsDefault=true
   desc="Your random range (usually a dynamic paramater in 0-1 space)"
   description=Optional - input your own RGB values here
   inputName=Option 3
   sortPriority=3
   previewValue=<0 0 1 0>
  }
  n4 = FunctionInput() {
   bUsePreviewValueAsDefault=true
   desc="Your random range (usually a dynamic paramater in 0-1 space)"
   description=Optional - input your own RGB values here
   inputName=Option 2
   sortPriority=2
   previewValue=<0 1 0 0>
  }
  n5 = FunctionInput() {
   bUsePreviewValueAsDefault=true
   desc="Your random range (usually a dynamic paramater in 0-1 space)"
   description=Optional - input your own RGB values here
   inputName=Option 1
   sortPriority=1
   previewValue=<1 0 0 0>
  }
  n6 = If(A=@n2, B=Constant() {r=0.66}, "A > B"=@n3, "A == B"=@n3, "A < B"=If(A=@n2, "A > B"=@n4, "A == B"=@n5, "A < B"=@n5))
  FunctionOutput(None=DotProduct(A=FunctionInput() {bUsePreviewValueAsDefault=true, desc=The channel packed texture, inputName=Texture Input}, B=@n6))
  FunctionOutput(None=@n6) {
   description="If you just want the random pick between R,G, and B, use this output. Doesn't require an texture input."
   outputName=Vec3 Out Only
   sortPriority=1
  }
 ]
}
