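# Material function that generates random per-tile noise from UV coordinates for effects materials, in binary and continuous forms.
# It floors the UVs, then hashes each cell with a dot product, sine, seed multiply and frac.
# A density bias is subtracted before optional rounding.
# A second-order noise hashes another UV set with a slightly scaled seed, multiplied by the first noise unless a static switch bypasses that crosstalk.
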MaterialFunction /Game/Effects/Fort_Effects/Materials/_Functions/MF_RandomUVSample
graph {
 properties {bExposeToLibrary=true}
 nodes [
  n1 = TextureCoordinate() {uTiling=3, vTiling=3}
  n2 = DotProduct(
   A=Floor(None=FunctionInput(Preview=@n1) {bUsePreviewValueAsDefault=true, inputName=PrimaryTextureCoordinates, inputType=FunctionInput_Vector2})
   B=Constant2Vector() {g=31.2257, r=7.3322}
  )
  n3 = FunctionInput() {bUsePreviewValueAsDefault=true, inputName=RandomSeed, inputType=FunctionInput_Scalar, sortPriority=4, previewValue=<7 0 0 1>}
  n4 = Frac(None=Multiply(A=Sine(None=@n2), B=@n3) {constB=43758.5469})
  n5 = FunctionInput() {
   bUsePreviewValueAsDefault=true
   description=Bias the amount of filled vs. empty noise tiles
   inputName=NoiseDensityBias
   inputType=FunctionInput_Scalar
   sortPriority=4
  }
  n6 = Subtract(A=@n4, B=@n5)
  FunctionOutput(None=ComponentMask(None=Round(None=@n6)) {r=true}) {outputName=Noise}
  FunctionOutput(None=ComponentMask(None=@n6) {r=true}) {description=Noise but not binary i.e. values other than just 0 or 1, outputName=NoiseWithRange, sortPriority=1}
  n7 = FunctionInput(Preview=TextureCoordinate() {uTiling=5, vTiling=5}) {
   bUsePreviewValueAsDefault=true
   description="Used for the randomization of blocks, use different coordinates than the first node"
   inputName=2ndOrderTextureCoordinates
   inputType=FunctionInput_Vector2
   sortPriority=1
  }
  n8 = DotProduct(A=Floor(None=@n7), B=Constant2Vector() {g=22.3365, r=8.1126})
  n9 = Frac(None=Multiply(A=Sine(None=@n8), B=Multiply(A=@n3) {constB=0.9953}) {constB=43758.5469})
  n10 = FunctionInput(Preview=StaticBool()) {
   bUsePreviewValueAsDefault=true
   description="First noise no longer influences 2nd noise, makes second order noise more independent"
   inputName=Bypass Noise Influence Crosstalk
   inputType=FunctionInput_StaticBool
   sortPriority=6
  }
  n11 = Subtract(A=StaticSwitch(True=@n9, False=Multiply(A=@n4, B=@n9), Value=@n10), B=@n5)
  FunctionOutput(None=ComponentMask(None=Round(None=@n11)) {r=true}) {description=2 different texture coordinates are used to create more noise randomness, outputName=2ndOrderNoise, sortPriority=2}
  FunctionOutput(None=ComponentMask(None=@n11) {r=true}) {description=Noise but not binary i.e. values other than just 0 or 1, outputName=2ndOrderNoiseWithRange, sortPriority=3}
 ]
}
