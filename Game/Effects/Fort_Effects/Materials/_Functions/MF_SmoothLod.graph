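# Material function that provides a smooth distance LOD for sprite particles, fading opacity and rescaling them at near and far camera ranges.
# Two vertex-shader CameraDepthFade calls, shaped by power curves, drive biased opacity ramps, output separately and multiplied together.
# The same fades blend near, mid and far scale multipliers on particle size.
# The WPO output rebuilds each vertex's offset from the particle center at that size, letting sprites shrink away.
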MaterialFunction /Game/Effects/Fort_Effects/Materials/_Functions/MF_SmoothLod
graph {nodes [
  n1 = FunctionInput() {
   bUsePreviewValueAsDefault=true
   description=Negative Values Bring the Opacity Transition towards the camera to start fading beore the vertex shader transition
   inputName=Opacity Near Distance Bias
   inputType=FunctionInput_Scalar
   sortPriority=10
  }
  n2 = FunctionInput() {
   bUsePreviewValueAsDefault=true
   description=Transition Region when close to the camera
   inputName=NearFadeLength
   inputType=FunctionInput_Scalar
   previewValue=<128 0 0 1>
  }
  n3 = FunctionInput() {
   bUsePreviewValueAsDefault=true
   description="Near Fade offset (Bias where no transition happens)"
   inputName=NearFadeOffset
   inputType=FunctionInput_Scalar
   sortPriority=1
  }
  n4 = MaterialFunctionCall(Fade Length=@n2, Fade Offset=@n3, For Vertex Shader=StaticBool() {value=true}) {materialFunction=&CameraDepthFade}
  n5 = FunctionInput() {
   bUsePreviewValueAsDefault=true
   description="Power function on the fade curve, so the transition region is nonlinear"
   inputName=NearFadeCurve
   inputType=FunctionInput_Scalar
   sortPriority=2
   previewValue=<1 0 0 1>
  }
  n6 = Clamp(None=ComponentMask(None=Power(Base=@n4.Result, Exp=@n5)) {r=true})
  n7 = LinearInterpolate(A=@n1, Alpha=@n6)
  FunctionOutput(None=Clamp(None=@n7)) {outputName=NearFadeOnly, sortPriority=2}
  n8 = FunctionInput() {
   bUsePreviewValueAsDefault=true
   description=Negative Values Bring the Opacity Transition towards the camera to start fading beore the vertex shader transition
   inputName=Opacity Far Distance Bias
   inputType=FunctionInput_Scalar
   sortPriority=11
  }
  n9 = FunctionInput() {
   bUsePreviewValueAsDefault=true
   description=Transition Region when far from the camera
   inputName=FarFadeLength
   inputType=FunctionInput_Scalar
   sortPriority=5
   previewValue=<128 0 0 1>
  }
  n10 = FunctionInput() {
   bUsePreviewValueAsDefault=true
   description="Far Fade offset (Bias where no transition happens)"
   inputName=FarFadeOffset
   inputType=FunctionInput_Scalar
   sortPriority=6
   previewValue=<512 0 0 1>
  }
  n11 = MaterialFunctionCall(Fade Length=@n9, Fade Offset=@n10, For Vertex Shader=StaticBool() {value=true}) {materialFunction=&CameraDepthFade}
  n12 = FunctionInput() {
   bUsePreviewValueAsDefault=true
   description="Power function on the fade curve, so the transition region is nonlinear"
   inputName=FarFadeCurve
   inputType=FunctionInput_Scalar
   sortPriority=7
   previewValue=<1 0 0 1>
  }
  n13 = Clamp(None=ComponentMask(None=Power(Base=@n11.Result, Exp=@n12)) {r=true})
  n14 = LinearInterpolate(B=@n8, Alpha=@n13) {constA=1, constB=0}
  FunctionOutput(None=Clamp(None=@n14)) {outputName=FarFadeOnly, sortPriority=3}
  FunctionOutput(None=Clamp(None=Multiply(A=@n7, B=@n14))) {outputName=Opacity, sortPriority=1}
  n15 = ParticleSize()
  n16 = FunctionInput() {
   bUsePreviewValueAsDefault=true
   description=How small the particle get when fully near
   inputName=NearFadeEndingScale
   inputType=FunctionInput_Scalar
   sortPriority=4
   previewValue=<-0.5 0 0 1>
  }
  n17 = FunctionInput() {
   bUsePreviewValueAsDefault=true
   description="Multiplier on the particle size, as you start to get close to a particle."
   inputName=NearFadeStartingScale
   inputType=FunctionInput_Scalar
   sortPriority=3
   previewValue=<1 0 0 1>
  }
  n18 = FunctionInput() {
   bUsePreviewValueAsDefault=true
   description=Scale on the particle when fully far away
   inputName=FarFadeEndingScale
   inputType=FunctionInput_Scalar
   sortPriority=8
   previewValue=<-0.5 0 0 1>
  }
  n19 = LinearInterpolate(A=LinearInterpolate(A=@n16, B=@n17, Alpha=@n6), B=@n18, Alpha=@n13)
  n20 = Clamp(None=Multiply(A=ComponentMask(None=@n15) {r=true}, B=Clamp(None=@n19) {clampMode=CMODE_ClampMin, minDefault=-1})) {clampMode=CMODE_ClampMin}
  n21 = Subtract(A=WorldPosition().XYZ, B=ParticlePositionWS())
  n22 = Normalize(VectorInput=@n21)
  FunctionOutput(None=Add(A=Multiply(A=@n20, B=@n22), B=OneMinus(None=@n21))) {description=Connect in World position offset in a sprite material, outputName=WPO}
 ]}
