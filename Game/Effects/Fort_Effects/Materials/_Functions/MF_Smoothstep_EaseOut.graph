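# Material function for effects materials that reshapes a linear input, described as running from 1 to 0, along an eased curve.
# Despite the Smoothstep name, it uses a sine: the input is scaled and offset to cover a quarter period.
# The sine result goes through a bias-scale and a final multiply to set the output range.
# A looping, normalized Time ramp serves as the input's preview and default.
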MaterialFunction /Game/Effects/Fort_Effects/Materials/_Functions/MF_Smoothstep_EaseOut
graph {nodes [
  n1 = FunctionInput(Preview=Divide(A=Time() {bOverride_Period=true, period=3}) {constB=3}) {bUsePreviewValueAsDefault=true, description=Input must transition from 1 to 0}
  n2 = Sine(None=Add(A=Multiply(A=@n1) {constB=0.25}) {constB=0.5})
  FunctionOutput(None=Multiply(A=ConstantBiasScale(None=@n2)) {constB=4})
 ]}
