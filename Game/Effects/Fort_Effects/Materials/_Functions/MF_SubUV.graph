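# VFX material function that plays a SubUV flipbook from a texture atlas with configurable column and row counts.
# Speed-scaled time, either automatic with a one-second period or manually supplied, selects the column and row through Frac and Floor.
# Scaled UVs offset to that cell sample the texture, outputting RGB, alpha and UVs.
# Despite its description mentioning linear blending, it takes one sample without blending between frames.
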MaterialFunction /Game/Effects/Fort_Effects/Materials/_Functions/MF_SubUV
graph {
 properties {bExposeToLibrary=true, description=Allows subUVs in a material with linear blending between frames, libraryCategoriesText [Misc,VFX]}
 nodes [
  n1 = FunctionInput() {
   bUsePreviewValueAsDefault=true
   description=Speed up or slow down the time expression through a multiply
   inputName=Time Speed
   inputType=FunctionInput_Scalar
   sortPriority=4
   previewValue=<1 0 0 0>
  }
  n2 = FunctionInput() {
   bUsePreviewValueAsDefault=true
   description="Input either a time expression, or control with a curve through a dynamic parameter"
   inputName=Time
   inputType=FunctionInput_Scalar
   sortPriority=6
  }
  n3 = StaticBool()
  n4 = Multiply(
   A=@n1
   B=StaticSwitch(
    True=@n2
    False=Time() {period=1}
    Value=FunctionInput(Preview=@n3) {bUsePreviewValueAsDefault=true, inputName=Manual Time Control, inputType=FunctionInput_StaticBool, sortPriority=5}
   )
  )
  n5 = FunctionInput() {
   bUsePreviewValueAsDefault=true
   description=The amount of images on U
   inputName=U_Images
   inputType=FunctionInput_Scalar
   sortPriority=1
   previewValue=<8 0 0 0>
  }
  n6 = Floor(None=Multiply(A=Frac(None=@n4), B=@n5))
  n7 = Divide(B=@n5) {constB=1}
  n8 = FunctionInput(Preview=TextureCoordinate()) {bUsePreviewValueAsDefault=true, inputName=UVs, inputType=FunctionInput_Vector2, sortPriority=7}
  n9 = ComponentMask(None=@n8) {r=true}
  n10 = ComponentMask(None=@n8) {g=true}
  n11 = FunctionInput() {
   bUsePreviewValueAsDefault=true
   description=The amount of images on V
   inputName=V_Images
   inputType=FunctionInput_Scalar
   sortPriority=2
   previewValue=<8 0 0 0>
  }
  n12 = Divide(B=@n11) {constB=1}
  n13 = Divide(A=@n4, B=@n11)
  n14 = Floor(None=Multiply(A=Frac(None=@n13), B=@n11))
  n15 = AppendVector(A=Add(A=Divide(A=@n6, B=@n5), B=Multiply(A=@n7, B=@n9)), B=Add(A=Multiply(A=@n10, B=@n12), B=Divide(A=@n14, B=@n11)))
  FunctionOutput(None=@n15) {description=Just the UVs, outputName=UVs, sortPriority=2}
  n16 = FunctionInput(Preview=TextureObject() {samplerType=SAMPLERTYPE_LinearColor, texture=/Game/Effects/Fort_Effects/Gameplay/SmokeDustExplosions/Textures/T_SmokePoof_8x8_01}) {
   bUsePreviewValueAsDefault=true
   description="Input subUV texture here (must be a texture object)"
   inputName=SubUV Texture RGB
   inputType=FunctionInput_Texture2D
   sortPriority=3
  }
  n17 = TextureSample(UVs=@n15, Tex=@n16) {samplerType=SAMPLERTYPE_LinearColor}
  FunctionOutput(None=@n17.RGB) {outputName=RGB}
  FunctionOutput(None=@n17.A) {outputName=A, sortPriority=1}
 ]
}
