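# Library-exposed effects material function that scales and pans UVs over time, then samples a texture with them.
# InUV is multiplied by the red channel of InRGB on both axes, leaving green unused, then offset by blue and InA as U and V speeds times InTime.
# It outputs the panned UV and sampled RGB.
# All inputs have preview defaults, including Time and a noise texture set to normal-map sampling.
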MaterialFunction /Game/Effects/Fort_Effects/Materials/_Functions/MF_TexturePanner
graph {
 properties {bExposeToLibrary=true}
 nodes [
  n1 = TextureCoordinate()
  n2 = Constant4Vector() {constant=<rgba 1 1 1 0>}
  n3 = MaterialFunctionCall(Float3=FunctionInput(Preview=@n2.RGBA) {bUsePreviewValueAsDefault=true, inputName=InRGB, sortPriority=2}) {materialFunction=&BreakOutFloat3Components}
  n4 = FunctionInput(Preview=@n2.A) {bUsePreviewValueAsDefault=true, inputName=InA, inputType=FunctionInput_Scalar, sortPriority=3}
  n5 = Time()
  n6 = Add(
   A=Multiply(
    A=FunctionInput(Preview=@n1) {bUsePreviewValueAsDefault=true, inputName=InUV, inputType=FunctionInput_Vector2}
    B=MaterialFunctionCall(R=@n3.R, G=@n3.R) {materialFunction=&AppendMany}.RG
   )
   B=Multiply(
    A=MaterialFunctionCall(R=@n3.B, G=@n4) {materialFunction=&AppendMany}.RG
    B=FunctionInput(Preview=@n5) {bUsePreviewValueAsDefault=true, inputName=InTime, inputType=FunctionInput_Scalar, sortPriority=4}
   )
  )
  n7 = TextureObject() {samplerType=SAMPLERTYPE_Normal, texture=/Game/Global/Textures/Default/Noise/64x64Noise}
  FunctionOutput(None=TextureSample(
    UVs=@n6
    Tex=FunctionInput(Preview=@n7) {bUsePreviewValueAsDefault=true, inputName=Texture, inputType=FunctionInput_Texture2D, sortPriority=1}
   ).RGB) {outputName=Sampled, sortPriority=2}
  FunctionOutput(None=@n6) {outputName=UV, sortPriority=1}
 ]
}
