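# Material function in the FX_Common library that outputs a scalar driven by the current time of day.
# It breaks the FortniteActiveTimeOfDay vector from FortniteMaterialParameters into R, G, B and A weights for morning, day, evening and night.
# Each weight multiplies its scalar input, which has a preview default, and the four products are summed into a weighted blend.
# The description says "Color", but all four value inputs are scalars.
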MaterialFunction /Game/Effects/Fort_Effects/Materials/_Functions/Scalar_Time_of_Day
graph {
 properties {bExposeToLibrary=true, description=Color Based on Time of Day, libraryCategoriesText [Misc,FX_Common]}
 nodes [
  n1 = CollectionParameter() {
   collection=&FortniteMaterialParameters
   parameterName=FortniteActiveTimeOfDay
  }
  n2 = MaterialFunctionCall(Float4=@n1) {materialFunction=&BreakOutFloat4Components}
  n3 = Multiply(
   A=FunctionInput(Preview=Constant() {r=0.45}) {bUsePreviewValueAsDefault=true, inputName=Value_Morning, inputType=FunctionInput_Scalar}
   B=@n2.R
  )
  n4 = Multiply(
   A=FunctionInput(Preview=Constant()) {bUsePreviewValueAsDefault=true, inputName=Value_Day, inputType=FunctionInput_Scalar, sortPriority=1}
   B=@n2.G
  )
  n5 = Multiply(
   A=FunctionInput(Preview=Constant() {r=0.65}) {bUsePreviewValueAsDefault=true, inputName=Value_Evening, inputType=FunctionInput_Scalar, sortPriority=2}
   B=@n2.B
  )
  n6 = Multiply(
   A=FunctionInput(Preview=Constant() {r=1}) {bUsePreviewValueAsDefault=true, inputName=Value_Night, inputType=FunctionInput_Scalar, sortPriority=3}
   B=@n2.A
  )
  FunctionOutput(None=Add(A=Add(A=@n3, B=@n4), B=Add(A=@n5, B=@n6)))
 ]
}
