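# Material function that converts a 2D UV vector to radial coordinates, packing normalized angle and distance from the origin.
# Custom atan2 nodes give the angle, divided by two pi and wrapped with Frac, while the vector's length supplies the second channel.
# A reversed output negates X, moving the angle seam to the opposite side.
# A third output raises the reversed angle's cosine to a high power, masking a narrow band at that seam.
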MaterialFunction /Game/Effects/Fort_Effects/Materials/_Functions/VectortoRadialCorrection
graph {nodes [
  n1 = FunctionInput(Preview=Subtract(A=TextureCoordinate() {uTiling=2, vTiling=2})) {bUsePreviewValueAsDefault=true, inputName=UVs, inputType=FunctionInput_Vector2}
  MaterialExpressionComponentMask_25 = ComponentMask(None=@n1) {r=true}
  MaterialExpressionComponentMask_26 = ComponentMask(None=@n1) {g=true}
  n2 = Custom!(x=@MaterialExpressionComponentMask_25, y=@MaterialExpressionComponentMask_26) {
   code="return (atan2(y,x));"
   description=atan2
   outputType=CMOT_Float1
   inputs [
    {inputName=x, input {expression=&self:MaterialExpressionComponentMask_25}}
    {inputName=y, input {expression=&self:MaterialExpressionComponentMask_26}}
   ]
  }
  n3 = Divide(A=ComponentMask(None=@n2) {r=true}, B=Constant() {r=6.2832})
  n4 = Multiply(A=@n1, B=@n1)
  n5 = SquareRoot(None=Add(A=ComponentMask(None=@n4) {r=true}, B=ComponentMask(None=@n4) {g=true}))
  FunctionOutput(None=AppendVector(A=Frac(None=@n3), B=@n5) {desc=VectortoRadial}) {outputName=Vector to Radial}
  MaterialExpressionMultiply_26 = Multiply(A=ComponentMask(None=@n1) {r=true}) {constB=-1}
  MaterialExpressionComponentMask_24 = ComponentMask(None=@n1) {g=true}
  n6 = Custom(x=@MaterialExpressionMultiply_26, y=@MaterialExpressionComponentMask_24) {
   code="return (atan2(y,x));"
   description=atan2
   outputType=CMOT_Float1
   inputs [
    {inputName=x, input {expression=&self:MaterialExpressionMultiply_26}}
    {inputName=y, input {expression=&self:MaterialExpressionComponentMask_24}}
   ]
  }
  n7 = Frac(None=Divide(A=ComponentMask(None=@n6) {r=true}, B=Constant() {r=6.2832}))
  FunctionOutput(None=AppendVector(A=@n7, B=@n5) {desc=VectortoRadialReversed}) {outputName=Vector to Radial Reversed, sortPriority=1}
  n8 = Cosine(None=@n7)
  n9 = Power(Base=@n8) {constExponent=500}
  FunctionOutput(None=Clamp(None=Multiply(A=@n8, B=@n9)) {desc=Mask}) {outputName=Vector to Reversed Mask, sortPriority=2}
 ]}
